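# Modulate-blended unlit Niagara sprite material that fakes a colored light on scene surfaces for the Monarch emote effect.
# It measures the distance from the particle center to the world position behind the translucency, normalized by half the particle size, producing a spherical falloff sharpened to the fourth power.
# That falloff times particle alpha blends from white to particle color, hue-shifted increasingly toward the edge.
# Depth testing is disabled.
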Material /WinterDecay/Athena/Effects/Particles/Emotes/NS_Emote_Monarch/M_Emote_Monarch_FakeLight
graph {
 properties {
  bDisableDepthTest=true
  bUsedWithNiagaraSprites=true
  blendMode=BLEND_Modulate
  pixelDepthOffsetMode=PDOM_Legacy
  shadingModel=MSM_Unlit
  translucencyPass=MTP_BeforeDOF
  usedShadingModels=Unlit
 }
 nodes [
  n1 = ScalarParameter() {defaultValue=0.07, parameterName=Hue Shift Amount}
  n2 = MaterialFunctionCall() {materialFunction=&WorldPositionBehindTranslucency}
  n3 = ParticlePositionWS()
  n4 = ComponentMask(None=ParticleSize()) {r=true}
  n5 = Saturate(None=Divide(A=Distance(A=@n2.Result, B=@n3), B=Multiply(A=@n4) {constB=0.5}))
  n6 = ParticleColor()
  n7 = OneMinus(None=@n5)
  n8 = Multiply(A=@n7, B=@n7)
  LinearInterpolate(
   B=MaterialFunctionCall(Hue Shift Percentage=Multiply(A=@n1, B=@n5), Texture=@n6.RGB) {materialFunction=&HueShift}.Result
   Alpha=Multiply(A=@n6.A, B=Multiply(A=@n8, B=@n8))
  ) {constA=1}
 ]
}
